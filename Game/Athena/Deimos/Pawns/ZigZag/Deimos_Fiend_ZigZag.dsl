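; ZigZag variant of the Deimos fiend zombie pawn, handling death in water and its death dissolve visuals.
; Water body entry, water interaction and water navmesh entry each print a debug string and, if WaterDeath is set, force-kill the pawn as its own killer.
; On the death material event, remote instances put masked dynamic materials on mesh slots 1–3, stamped with game time, to hide the head.
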(event WaterBody|EventOnEnteredWaterBody
  (params (WaterBody :type Fort%20Water%20Body%20Actor%20Object%20Reference) (bIsFirstBody :type Boolean))
  ;[1 1 1] Zombies can't swim
  (|CallParentFunction WaterBody bIsFirstBody)
  (Development|PrintString "ZIGZAG AI ENTERED WATER BODY")
  (if (Variables|Default|GetWaterDeath)
    (FortPawn|ForceKill :KillerActor self)))

(event WaterBody|EventOnBeginWaterInteraction
  (params (WaterSurfaceLocation :type Vector :default "0, 0, 0") (WaterDepth :type Float%20%28single-precision%29))
  ;[1 1 1] Zombies can't swim
  (|CallParentFunction WaterSurfaceLocation WaterDepth)
  (Development|PrintString "ZIGZAG AI ENTERED WATER BODY")
  (if (Variables|Default|GetWaterDeath)
    (FortPawn|ForceKill :KillerActor self)))

(event WaterBody|EventOnWaterNavMeshEnter
  ;[1 1 1] Zombies can't swim
  (|CallParentFunction)
  (Development|PrintString "ZIGZAG AI ENTERED WATER BODY")
  (if (Variables|Default|GetWaterDeath)
    (FortPawn|ForceKill :KillerActor self)))

(event EventAnimateDeathMaterial
  (bind _returnvalue (Utilities|Time|GetGameTimeinSeconds))
  (bind _mesh (Variables|Character|GetMesh))
  (bind _maskeddeathmaterial (Variables|Default|GetMaskedDeathMaterial))
  ;[1 1 1] Hide zombie head on death
  (|CallParentFunction)
  (Networking|SwitchHasAuthority
    (:Remote
      (Development|PrintString "HIDE HEAD" false false)
      (bind _returnvalue_2 (Rendering|Material|CreateDynamicMaterialInstance _mesh 1 _maskeddeathmaterial "None" :!class /Script/Engine.PrimitiveComponent))
      (Rendering|Material|SetScalarParameterValue _returnvalue_2 "Death_CurrentTime" _returnvalue :!class /Script/Engine.MaterialInstanceDynamic)
      (bind _returnvalue_3 (Rendering|Material|CreateDynamicMaterialInstance _mesh 2 _maskeddeathmaterial "None" :!class /Script/Engine.PrimitiveComponent))
      (Rendering|Material|SetScalarParameterValue _returnvalue_3 "Death_CurrentTime" _returnvalue :!class /Script/Engine.MaterialInstanceDynamic)
      (bind _returnvalue_4 (Rendering|Material|CreateDynamicMaterialInstance _mesh 3 _maskeddeathmaterial "None" :!class /Script/Engine.PrimitiveComponent))
      (Rendering|Material|SetScalarParameterValue _returnvalue_4 "Death_CurrentTime" _returnvalue :!class /Script/Engine.MaterialInstanceDynamic))))

(fn ConstructionScript
  (|CallParentFunction))
